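# Additive unlit material that gives the Honeycomb vehicle wheel an animated hologram look, with separate rim and tire layers blended by vertex color red.
# The rim samples a geometric texture through panning UVs offset by its hologram texture's channels.
# The tire feeds similarly offset panning UVs into a sine, sharpened with a power term, for procedural bands.
# Both share emissive color and intensity parameters and take opacity from their mask textures' green channels.
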Material /VehicleCosmetics/Wheels/Honeycomb/Materials/M_Wheel_Honetcomb_Hologram
graph {
 properties {
  bAutomaticallySetUsageInEditor=false
  bUsedWithMeshDeformer=true
  bUsedWithSkeletalMesh=true
  blendMode=BLEND_Additive
  shadingModel=MSM_Unlit
  usedShadingModels=Unlit
 }
 nodes [
  n1 = TextureSample() {samplerSource=SSM_Wrap_WorldGroupSettings, texture=/VehicleCosmetics/Wheels/Honeycomb/Textures/T_Rim_Honeycomb_Hologram}
  n2 = Add(A=Panner(Coordinate=TextureCoordinate()) {speedX=0.35}, B=@n1.R)
  n3 = TextureSample(UVs=Add(A=Add(A=@n2, B=@n1.G), B=@n1.B)) {samplerSource=SSM_Wrap_WorldGroupSettings, texture=/VehicleCosmetics/Textures/UniversalTextures/T_Universal_Geometric}
  n4 = VectorParameter() {group=1_Rim_Property, parameterName=EmissiveColor, defaultValue=<rgba 1 0.35 0 1>}
  n5 = ScalarParameter() {defaultValue=8, group=1_Rim_Property, parameterName=EmissiveIntensity}
  n6 = Multiply(A=Multiply(A=ComponentMask(None=@n3.RGB) {r=true}, B=@n4.RGB), B=@n5) {constB=8}
  n7 = TextureSampleParameter2D() {
   group=1_Rim_Property
   parameterName=Rim_Opacity_Hologram
   samplerSource=SSM_Wrap_WorldGroupSettings
   samplerType=SAMPLERTYPE_Masks
   texture=/VehicleCosmetics/Wheels/Honeycomb/Textures/T_Rim_Honeycomb_RGB
  }
  n8 = TextureCoordinate()
  n9 = TextureSampleParameter2D() {
   group=2_Tire_Property
   parameterName=Tire_RGBA_Hologram
   samplerSource=SSM_Wrap_WorldGroupSettings
   texture=/VehicleCosmetics/Wheels/Honeycomb/Textures/T_Tire_Honeycomb_Hologram
  }
  n10 = Add(A=Panner(Coordinate=Divide(A=@n8) {constB=1}) {speedY=0.3}, B=@n9.R)
  n11 = Sine(None=Add(A=Add(A=@n10, B=@n9.G), B=@n9.B))
  n12 = Power(Base=@n11) {constExponent=12}
  n13 = Saturate(None=Add(A=Divide(A=@n11) {constB=6}, B=Divide(A=@n12) {constB=4}))
  n14 = Multiply(A=Multiply(A=ComponentMask(None=@n13) {g=true}, B=@n4.RGB), B=@n5) {constB=8}
  n15 = TextureSampleParameter2D() {
   group=2_Tire_Property
   parameterName=Tire_Opacity_Hologram
   samplerSource=SSM_Wrap_WorldGroupSettings
   samplerType=SAMPLERTYPE_Masks
   texture=/VehicleCosmetics/Wheels/Honeycomb/Textures/T_Tire_Honeycomb_RGB
  }
  n16 = VertexColor()
  GetMaterialAttributes(None=BlendMaterialAttributes(
    A=SetMaterialAttributes!(MaterialAttributes, Emissive Color=@n6, Opacity=@n7.G) {attributeSetTypes [#1, #2]}
    B=SetMaterialAttributes(Emissive Color=@n14, Opacity=@n15.G) {attributeSetTypes [#1, #2]}
    Alpha=NamedRerouteDeclaration(None=@n16.R) {name=VertColor_R_Out, nodeColor=<rgba 1 0.5 0 1>}
   )) {attributeGetTypes [#1, #2]}
 ]
}
